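# Niagara cosmetic effect for the Sleepy Time Ninja male medium character body, adding soft orange flares to its two grenades.
# GrenadeFlare steadily spawns short-lived, camera-offset sprites that fade in and out through an alpha curve.
# A socket reader follows the two dyn_grenade sockets each frame, with each particle's partition choosing its socket and positional offset.
# Flare and Light emitters list no renderers or modules; a bisHeadHidden user bool is exposed, but its use is not shown.
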NiagaraSystem /BRCosmetics/Characters/Player/Male/Medium/Bodies/M_MED_SleepyTimeNinja/FX/NS_MED_SleepyTime_Ninja
dependencies {
 usedDataInterfaces [&NiagaraDataInterfaceSocketReader, &NiagaraDataInterfaceArrayFloat3, &NiagaraDataInterfaceCurve]
 usedDynamicInputs [&MakeCustomFloatFromBool]
 usedModules [
  &SystemState
  &self:SystemSpawnScript.SystemScriptSource.SystemScriptGraph.NiagaraNodeAssignment_0.SetVariables_#1
 ]
}
emitters {
 Flare {rendererClasses [], renderers []}
 GrenadeFlare {
  emitterUpdateScript [
   &EmitterState()
   &SpawnRate(SpawnRate=3, Use Spawn Probability=false)
  ]
  particleSpawnScript [
   &PartitionParticles()
   &InitializeParticle(
    Color=<rgba 1 0.5262 0.0139 0.12>
    Lifetime=2
    Sprite Size Mode=::NewEnumerator3
    Uniform Sprite Size=50
    UsePositionOffset=false
    Write Material Random=true
   )
   &CameraOffset(Camera Offset Amount=10)
  ]
  particleUpdateScript [
   &ParticleState()
   &SocketLocation(
    SamplePosOffset=true
    Sampled Position Offset=&SelectVectorFromArray(
     Direct Array Index=$Particles.PartitionParticles.Partition
     Vector Selection Array=DataInterface() {propertyValues {FloatData [<1.5 0 -3>, <0 0 -2>]}}
    )
    Socket Index=$Particles.PartitionParticles.Partition
    Socket Reader=DataInterface() {propertyValues {
      AttachComponentClass=null
      AttachComponentTag=None
      EditorPreviewAsset=/BRCosmetics/Characters/Player/Male/Medium/Bodies/M_MED_SleepyTimeNinja/Meshes/M_MED_Sleepy_Time_Ninja
      SourceActorRef=null
      SourceAsset=null
      SourceMode=Default
      bRequireCurrentFrameData=true
      bUpdateSocketsPerFrame=true
      FilteredSockets [dyn_grenade_01_CNST, dyn_grenade_02_CNST]
      ObjectParameterBinding {parameter {name=None, typeDefHandle {registeredTypeIndex=1}, varData []}}
     }}
    Socket Sampling Mode=::NewEnumerator3
    VelocitySpeedLimit=false
   )
   &ScaleColor(
    Scale Alpha=&FloatFromCurve(FloatCurve=DataInterface() {propertyValues {Curve {keys [<at 0 0>,<at 0.3 1>,<at 1 0>]}}})
    ScaleA=true
    ScaleRGB=true
   )
  ]
  rendererClasses [&NiagaraSpriteRendererProperties]
  renderers [{rendererClass=&NiagaraSpriteRendererProperties}]
 }
 Light {rendererClasses [], renderers []}
}
summary {
 systemName=NS_MED_SleepyTime_Ninja
 emitters [
  {emitterName=Flare, rendererClasses []}
  {emitterName=Light, rendererClasses []}
  {emitterName=GrenadeFlare, rendererClasses [&NiagaraSpriteRendererProperties]}
 ]
 userVariables [User.bisHeadHidden {type=NiagaraBool}]
}
