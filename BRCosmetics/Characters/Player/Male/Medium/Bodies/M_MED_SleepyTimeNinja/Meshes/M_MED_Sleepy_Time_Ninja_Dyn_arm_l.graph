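# Physics asset for the left arm of the male medium SleepyTimeNinja body, setting up collision and joints for its three-bone shoulder pad chain.
# Capsules cover the upper and lower arm, while boxes form the pad segments, with two extra angled boxes on the last segment.
# The pad chain links to the upper arm with twist locked and mostly single-axis limited swing, the last joint allowing a second.
# The elbow constraint leaves all angular axes free.
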PhysicsAsset /BRCosmetics/Characters/Player/Male/Medium/Bodies/M_MED_SleepyTimeNinja/Meshes/M_MED_Sleepy_Time_Ninja_Dyn_arm_l
bodies {
 dyn_shoulderPad_l_01 [{shapeType=Box, extentX=9, extentY=8, extentZ=2, center=<5 0 0>, rotation=<rot 6 0 0>}]
 dyn_shoulderPad_l_02 [{shapeType=Box, extentX=9.282, extentY=8, extentZ=2, center=<4.8043 0 0.5112>, rotation=<rot 0.8 0 0>}]
 dyn_shoulderPad_l_03 [
  {shapeType=Box, extentX=9.6178, extentY=8, extentZ=2, center=<4.9371 0 0.1604>, rotation=<rot -5 0 0>}
  {shapeType=Box, extentX=6, extentY=5.5, extentZ=2, center=<3 5.5 -1.5>, rotation=<rot -6 0 50>}
  {shapeType=Box, extentX=6, extentY=5.5, extentZ=2, center=<3 -5.5 -1.5>, rotation=<rot -6 0 -50>}
 ]
 lowerarm_l [{shapeType=Capsule, length=22.3282, radius=3.9979, center=<13 0 0>, rotation=<rot 87 0 0>}]
 upperarm_l [{shapeType=Capsule, length=22.8149, radius=4.185, center=<13 0 0>, rotation=<rot 90 0 0>}]
}
constraints [
 {
  bone1Name=dyn_shoulderPad_l_01
  bone2Name=upperarm_l
  swing1LimitDegrees=5
  swing1Motion=Locked
  swing2LimitDegrees=30
  twistLimitDegrees=1
  twistMotion=Locked
 }
 {
  bone1Name=dyn_shoulderPad_l_02
  bone2Name=dyn_shoulderPad_l_01
  swing1LimitDegrees=5
  swing1Motion=Locked
  swing2LimitDegrees=30
  twistLimitDegrees=1
  twistMotion=Locked
 }
 {
  bone1Name=dyn_shoulderPad_l_03
  bone2Name=dyn_shoulderPad_l_02
  swing1LimitDegrees=10
  swing2LimitDegrees=30
  twistLimitDegrees=1
  twistMotion=Locked
 }
 {
  bone1Name=lowerarm_l
  bone2Name=upperarm_l
  swing1LimitDegrees=45
  swing1Motion=Free
  swing2LimitDegrees=45
  swing2Motion=Free
  twistLimitDegrees=45
  twistMotion=Free
 }
]
