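# Physics asset for the right arm dynamics of the Sleepy Time Ninja medium male player body, covering a three-segment shoulder pad.
# Box bodies shape each pad segment, with three angled boxes on the last, while capsules cover the upper arm twist and lower arm.
# Constraints chain the segments from the upper arm twist bone, locking twist throughout and limiting swing to mostly one axis, with only the last segment given limited secondary swing.
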PhysicsAsset /BRCosmetics/Characters/Player/Male/Medium/Bodies/M_MED_SleepyTimeNinja/Meshes/M_MED_Sleepy_Time_Ninja_Dyn_arm_r
bodies {
 dyn_shoulderPad_r_01 [{shapeType=Box, extentX=9.0548, extentY=7, extentZ=2, center=<5 0 0>, rotation=<rot 6 0 0>}]
 dyn_shoulderPad_r_02 [{shapeType=Box, extentX=7.5782, extentY=7, extentZ=2, center=<4 0 0.4921>, rotation=<rot 0.8 0 0>}]
 dyn_shoulderPad_r_03 [
  {shapeType=Box, extentX=8.9665, extentY=8, extentZ=2, center=<4.4159 0.0013 -0.0474>, rotation=<rot -6.5 0 0>}
  {shapeType=Box, extentX=6, extentY=5.5, extentZ=2, center=<3 5.5 -1.5>, rotation=<rot -6 0 50>}
  {shapeType=Box, extentX=6, extentY=5.5, extentZ=2, center=<3 -5.5 -1.5>, rotation=<rot -6 0 -50>}
 ]
 lowerarm_r [{shapeType=Capsule, length=21.8318, radius=3.6682, center=<-13 0 0>, rotation=<rot 87 0 0>}]
 upperarm_twist_01_r [{shapeType=Capsule, length=23.1032, radius=3.8968, center=<-13 0 0>, rotation=<rot 90 0 0>}]
}
constraints [
 {
  bone1Name=dyn_shoulderPad_r_01
  bone2Name=upperarm_twist_01_r
  swing1LimitDegrees=1
  swing1Motion=Locked
  swing2LimitDegrees=30
  twistLimitDegrees=1
  twistMotion=Locked
 }
 {
  bone1Name=dyn_shoulderPad_r_02
  bone2Name=dyn_shoulderPad_r_01
  swing1LimitDegrees=1
  swing1Motion=Locked
  swing2LimitDegrees=30
  twistLimitDegrees=1
  twistMotion=Locked
 }
 {
  bone1Name=dyn_shoulderPad_r_03
  bone2Name=dyn_shoulderPad_r_02
  swing1LimitDegrees=10
  swing2LimitDegrees=30
  twistLimitDegrees=1
  twistMotion=Locked
 }
]
