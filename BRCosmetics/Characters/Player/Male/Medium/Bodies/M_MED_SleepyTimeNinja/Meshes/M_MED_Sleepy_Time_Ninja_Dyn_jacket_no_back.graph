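# Physics asset for the Sleepy Time Ninja male medium body, providing secondary motion for the jacket and two dangling grenades.
# Box-body chains form left, center and right panels, three segments in front and two in back, hung from spine_04 with limited swing and twist.
# Tight cross-constraints tie side hem segments to center ones.
# Grenades pivot on master bodies linked to spine_04; pelvis, spine and thigh capsules supply body volume with free angular joints.
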PhysicsAsset /BRCosmetics/Characters/Player/Male/Medium/Bodies/M_MED_SleepyTimeNinja/Meshes/M_MED_Sleepy_Time_Ninja_Dyn_jacket_no_back
bodies {
 dyn_grenade_01 [{shapeType=Capsule, length=5, radius=3.5, center=<3.5 0 0>}]
 dyn_grenade_01_MSTR [{shapeType=Capsule, length=1, radius=1}]
 dyn_grenade_02 [{shapeType=Capsule, length=5, radius=3.5, center=<3.5 0 0>}]
 dyn_grenade_02_MSTR [{shapeType=Capsule, length=1, radius=1}]
 dyn_jacket_bk_02 [{shapeType=Box, extentX=7.5, extentY=11, extentZ=3, center=<3.5 0 0>}]
 dyn_jacket_bk_03 [{shapeType=Box, extentX=14, extentY=11, extentZ=3, center=<6 0 0>}]
 dyn_jacket_bk_l_02 [{shapeType=Box, extentX=7.5, extentY=11, extentZ=3, center=<3.5 0 0>}]
 dyn_jacket_bk_l_03 [{shapeType=Box, extentX=9, extentY=11, extentZ=3, center=<4 0 0>}]
 dyn_jacket_bk_r_02 [{shapeType=Box, extentX=7.5, extentY=11, extentZ=3, center=<3.5 0 0>}]
 dyn_jacket_bk_r_03 [{shapeType=Box, extentX=9, extentY=11, extentZ=3, center=<4 0 0>}]
 dyn_jacket_fn_01 [{shapeType=Box, extentX=7.5, extentY=11, extentZ=3, center=<3.5 0 0>}]
 dyn_jacket_fn_02 [{shapeType=Box, extentX=7.5, extentY=11, extentZ=3, center=<3.5 0 0>}]
 dyn_jacket_fn_03 [{shapeType=Box, extentX=14, extentY=11, extentZ=3, center=<6 0 0>}]
 dyn_jacket_fn_l_01 [{shapeType=Box, extentX=7.5, extentY=11, extentZ=3, center=<3.5 0 0>}]
 dyn_jacket_fn_l_02 [{shapeType=Box, extentX=7.5, extentY=11, extentZ=3, center=<3.5 0 0>}]
 dyn_jacket_fn_l_03 [{shapeType=Box, extentX=9, extentY=11, extentZ=3, center=<4 0 0>}]
 dyn_jacket_fn_r_01 [{shapeType=Box, extentX=7.5, extentY=11, extentZ=3, center=<3.5 0 0>}]
 dyn_jacket_fn_r_02 [{shapeType=Box, extentX=7.5, extentY=11, extentZ=3, center=<3.5 0 0>}]
 dyn_jacket_fn_r_03 [{shapeType=Box, extentX=9, extentY=11, extentZ=3, center=<4 0 0>}]
 pelvis [{shapeType=Capsule, length=11.2992, radius=8.5101, center=<0 -0.5496 -0.0001>}]
 spine_01 [{shapeType=Capsule, length=9.9612, radius=7.9188, center=<0 0.3856 -0.0001>}]
 spine_02 [{shapeType=Capsule, length=9.8249, radius=7.6567, center=<0 0.8417 0>}]
 spine_03 [{shapeType=Capsule, length=10.9767, radius=8.4127, center=<0 0.779 0.0001>}]
 spine_04 [{shapeType=Capsule, length=12.2014, radius=9.6383, center=<-0.0001 0.6885 0.0001>}]
 thigh_twist_01_l [{shapeType=Capsule, length=30, radius=7.5, center=<5 0 0>, rotation=<rot 90 0 0>}]
 thigh_twist_01_r [{shapeType=Capsule, length=30, radius=7.5, center=<-5 0 0>, rotation=<rot 90 0 0>}]
}
constraints [
 {bone1Name=dyn_jacket_fn_01, bone2Name=spine_04, swing1LimitDegrees=30, swing2LimitDegrees=45, twistLimitDegrees=5}
 {bone1Name=dyn_jacket_fn_02, bone2Name=dyn_jacket_fn_01, swing1LimitDegrees=30, swing2LimitDegrees=45, twistLimitDegrees=5}
 {bone1Name=dyn_jacket_fn_03, bone2Name=dyn_jacket_fn_02, swing1LimitDegrees=30, swing2LimitDegrees=45, twistLimitDegrees=5}
 {bone1Name=dyn_jacket_fn_l_01, bone2Name=spine_04, swing1LimitDegrees=30, swing2LimitDegrees=45, twistLimitDegrees=5}
 {bone1Name=dyn_jacket_fn_l_02, bone2Name=dyn_jacket_fn_l_01, swing1LimitDegrees=30, swing2LimitDegrees=45, twistLimitDegrees=5}
 {bone1Name=dyn_jacket_fn_l_03, bone2Name=dyn_jacket_fn_l_02, swing1LimitDegrees=30, swing2LimitDegrees=45, twistLimitDegrees=5}
 {bone1Name=dyn_jacket_fn_r_01, bone2Name=spine_04, swing1LimitDegrees=15, swing2LimitDegrees=20, twistLimitDegrees=5}
 {bone1Name=dyn_jacket_fn_r_02, bone2Name=dyn_jacket_fn_r_01, swing1LimitDegrees=30, swing2LimitDegrees=45, twistLimitDegrees=5}
 {bone1Name=dyn_jacket_fn_r_03, bone2Name=dyn_jacket_fn_r_02, swing1LimitDegrees=30, swing2LimitDegrees=45, twistLimitDegrees=5}
 {bone1Name=dyn_jacket_bk_l_02, bone2Name=spine_04, swing1LimitDegrees=20, swing2LimitDegrees=30, twistLimitDegrees=5}
 {bone1Name=dyn_jacket_bk_l_03, bone2Name=dyn_jacket_bk_l_02, swing1LimitDegrees=20, swing2LimitDegrees=30, twistLimitDegrees=5}
 {bone1Name=dyn_jacket_bk_02, bone2Name=spine_04, swing1LimitDegrees=20, swing2LimitDegrees=30, twistLimitDegrees=5}
 {bone1Name=dyn_jacket_bk_03, bone2Name=dyn_jacket_bk_02, swing1LimitDegrees=20, swing2LimitDegrees=30, twistLimitDegrees=5}
 {bone1Name=dyn_jacket_bk_r_02, bone2Name=spine_04, swing1LimitDegrees=20, swing2LimitDegrees=30, twistLimitDegrees=5}
 {bone1Name=dyn_jacket_bk_r_03, bone2Name=dyn_jacket_bk_r_02, swing1LimitDegrees=20, swing2LimitDegrees=30, twistLimitDegrees=5}
 {
  bone1Name=spine_04
  bone2Name=spine_03
  swing1LimitDegrees=45
  swing1Motion=Free
  swing2LimitDegrees=45
  swing2Motion=Free
  twistLimitDegrees=45
  twistMotion=Free
 }
 {
  bone1Name=spine_02
  bone2Name=spine_01
  swing1LimitDegrees=45
  swing1Motion=Free
  swing2LimitDegrees=45
  swing2Motion=Free
  twistLimitDegrees=45
  twistMotion=Free
 }
 {
  bone1Name=spine_03
  bone2Name=spine_02
  swing1LimitDegrees=45
  swing1Motion=Free
  swing2LimitDegrees=45
  swing2Motion=Free
  twistLimitDegrees=45
  twistMotion=Free
 }
 {
  bone1Name=spine_01
  bone2Name=pelvis
  swing1LimitDegrees=45
  swing1Motion=Free
  swing2LimitDegrees=45
  swing2Motion=Free
  twistLimitDegrees=45
  twistMotion=Free
 }
 {
  bone1Name=thigh_twist_01_l
  bone2Name=pelvis
  swing1LimitDegrees=45
  swing1Motion=Free
  swing2LimitDegrees=45
  swing2Motion=Free
  twistLimitDegrees=45
  twistMotion=Free
 }
 {
  bone1Name=thigh_twist_01_r
  bone2Name=pelvis
  swing1LimitDegrees=45
  swing1Motion=Free
  swing2LimitDegrees=45
  swing2Motion=Free
  twistLimitDegrees=45
  twistMotion=Free
 }
 {bone1Name=dyn_jacket_bk_r_03, bone2Name=dyn_jacket_bk_03, swing1LimitDegrees=5, swing2LimitDegrees=10, twistLimitDegrees=5}
 {bone1Name=dyn_jacket_bk_l_03, bone2Name=dyn_jacket_bk_03, swing1LimitDegrees=5, swing2LimitDegrees=10, twistLimitDegrees=5}
 {bone1Name=dyn_jacket_fn_l_03, bone2Name=dyn_jacket_fn_03, swing1LimitDegrees=5, swing2LimitDegrees=10, twistLimitDegrees=5}
 {bone1Name=dyn_jacket_fn_r_03, bone2Name=dyn_jacket_fn_03, swing1LimitDegrees=5, swing2LimitDegrees=10, twistLimitDegrees=5}
 {bone1Name=dyn_grenade_01, bone2Name=dyn_grenade_01_MSTR, swing1LimitDegrees=10, swing2LimitDegrees=15, twistLimitDegrees=10}
 {bone1Name=dyn_grenade_02, bone2Name=dyn_grenade_02_MSTR, swing1LimitDegrees=10, swing2LimitDegrees=15, twistLimitDegrees=10}
 {
  bone1Name=dyn_grenade_01_MSTR
  bone2Name=spine_04
  swing1LimitDegrees=45
  swing1Motion=Free
  swing2LimitDegrees=45
  swing2Motion=Free
  twistLimitDegrees=45
  twistMotion=Free
 }
 {
  bone1Name=dyn_grenade_02_MSTR
  bone2Name=spine_04
  swing1LimitDegrees=45
  swing1Motion=Free
  swing2LimitDegrees=45
  swing2Motion=Free
  twistLimitDegrees=45
  twistMotion=Free
 }
]
