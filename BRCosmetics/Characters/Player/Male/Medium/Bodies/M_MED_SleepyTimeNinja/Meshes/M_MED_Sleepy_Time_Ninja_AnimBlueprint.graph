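# Character-part AnimBlueprint for the Sleepy Time Ninja body, adding secondary motion to jacket, collar, grenades, belt and shoulder pads.
# It copies the parent pose, then applies constraints and bone offsets driven by head and arm angles.
# Cached poses feed arm RigidBody sims, bypassed for custom animations, then one of two jacket sims chosen by backbling.
# Updates damp physics when targeting or crouching, reduce jacket gravity while skydiving before parachute deployment.
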AnimBlueprint /BRCosmetics/Characters/Player/Male/Medium/Bodies/M_MED_SleepyTimeNinja/Meshes/M_MED_Sleepy_Time_Ninja_AnimBlueprint
target_skeleton=&/M_MED_Sleepy_Time_Ninja_Skeleton
graphs [
 AnimGraph {nodes [
   n1 = Break Fort Anim Input Skydiving External Force(FortAnimInput_SkydivingExternalForce=Get SkydivingExternalForceInput()) -> (
    bUseSkydivingVectorForce=~
    SkydivingVectorForce=~
    HeadToPelvisDirection=~
    FloatingMultiplier=~
    FloatingAdditive=~
    DivingMultiplier=~
    DivingAdditive=~
    ParachutingMultiplier=~
    ParachutingAdditive=~
   )
   n2 = Get rigid alpha() -> (rigid alpha:double)
   n3 = Get gravity_override() -> (gravity_override:Vector)
   n4 = RigidBody!(
    ComponentPose=Local To Component(LocalPose=Use cached pose 'preJacketCache'())
    Alpha=@n2
    OverrideWorldGravity:Vector=@n3
    ExternalForce=@n1.SkydivingVectorForce
   ) -> (Pose) {
    bClampLinearTranslationLimitToRefPose=true
    bOverrideWorldGravity=true
    overridePhysicsAsset=&/M_MED_Sleepy_Time_Ninja_Dyn
    simulationSpace=BaseBoneSpace
    alphaBoolBlend {blendOption=Linear}
    baseBoneRef {boneName=spine_04}
    componentAppliedLinearAccClamp=<1000 1000 1500>
    componentLinearAccScale=<0.2 0.2 0.25>
    componentLinearVelScale=<0.2 0.2 0.2>
    overrideWorldGravity {z=-980}
   }
   n5 = RigidBody(
    ComponentPose=Local To Component(LocalPose=Use cached pose 'preJacketCache'())
    Alpha=@n2
    OverrideWorldGravity=@n3
    ExternalForce=@n1.SkydivingVectorForce
   ) {
    bClampLinearTranslationLimitToRefPose=true
    bOverrideWorldGravity=true
    overridePhysicsAsset=&/M_MED_Sleepy_Time_Ninja_Dyn_jacket_no_back
    simulationSpace=BaseBoneSpace
    alphaBoolBlend {blendOption=Linear}
    baseBoneRef {boneName=spine_04}
    componentAppliedLinearAccClamp=<1500 1500 1500>
    componentLinearAccScale=<0.3 0.3 0.3>
    componentLinearVelScale=<0.2 0.2 0.2>
    overrideWorldGravity {z=-980}
   }
   Output Pose(Result=Blend Poses by bool(BlendPose_0=Component To Local(ComponentPose=@n5), BlendPose_1=Component To Local(ComponentPose=@n4), bActiveValue=Get isWearingBackbling()) {blendType=Linear}) {layerGroup=None}
   n6 = "Transform (Modify) Bone - Bone: dyn_shoulderPad_l_01"(
    ComponentPose=Local To Component(LocalPose=Use cached pose 'PrePose'())
    Alpha=Get elbow_rotate_l() -> (elbow_rotate_l:double)
    Translation="-6.5,0,0"
    Rotation="0,0,0"
    Scale="1,1,1"
   ) {
    translationMode=BMM_Additive
    translationSpace=BCS_BoneSpace
    alphaBoolBlend {blendOption=Linear}
    boneToModify {boneName=dyn_shoulderPad_l_01}
    translation {x=-6.5}
   }
   n7 = "Transform (Modify) Bone - Bone: dyn_shoulderPad_r_01"(ComponentPose=@n6, Alpha=Get elbow_rotate_r() -> (elbow_rotate_r:double), Translation="-6.5,0,0", Rotation="0,0,0", Scale="1,1,1") {
    translationMode=BMM_Additive
    translationSpace=BCS_BoneSpace
    alphaBoolBlend {blendOption=Linear}
    boneToModify {boneName=dyn_shoulderPad_r_01}
    translation {x=-6.5}
   }
   n8 = RigidBody!(ComponentPose=@n7, Alpha=@n2, ExternalForce=@n1.SkydivingVectorForce) -> (Pose) {
    bClampLinearTranslationLimitToRefPose=true
    bOverrideWorldGravity=true
    overridePhysicsAsset=&/M_MED_Sleepy_Time_Ninja_Dyn_arm_l
    simulationSpace=BaseBoneSpace
    alphaBoolBlend {blendOption=Linear}
    baseBoneRef {boneName=upperarm_l}
    componentAppliedLinearAccClamp=<2000 2000 2000>
    componentLinearAccScale=<0.3 0.3 0.3>
    componentLinearVelScale=<0.2 0.2 0.2>
    overrideWorldGravity {z=-980}
   }
   n9 = RigidBody(ComponentPose=@n8, Alpha=@n2, ExternalForce=@n1.SkydivingVectorForce) {
    bClampLinearTranslationLimitToRefPose=true
    bOverrideWorldGravity=true
    overridePhysicsAsset=&/M_MED_Sleepy_Time_Ninja_Dyn_arm_r
    simulationSpace=BaseBoneSpace
    alphaBoolBlend {blendOption=Linear}
    baseBoneRef {boneName=upperarm_r}
    componentAppliedLinearAccClamp=<2000 2000 2000>
    componentLinearAccScale=<0.3 0.3 0.3>
    componentLinearVelScale=<0.2 0.2 0.2>
    overrideWorldGravity {z=-980}
   }
   n10 = "Transform (Modify) Bone - Bone: dyn_shoulderPad_r_01"(ComponentPose=Local To Component(LocalPose=Use cached pose 'PrePose'()), Alpha=1, Translation="0,0,0", Rotation="-10,-15,0", Scale="1,1,1") {
    rotationMode=BMM_Additive
    rotationSpace=BCS_BoneSpace
    translationMode=BMM_Additive
    translationSpace=BCS_BoneSpace
    alphaBoolBlend {blendOption=Linear}
    boneToModify {boneName=dyn_shoulderPad_r_01}
    rotation {pitch=-10, yaw=-15}
   }
   Save cached pose 'preJacketCache'(Pose=Blend Poses by bool(
     BlendPose_0=Component To Local(ComponentPose=@n10)
     BlendPose_1=Component To Local(ComponentPose=@n9)
     bActiveValue=Get bCanPlayCustomAnimations() -> (bCanPlayCustomAnimations:bool)
    ) {blendType=Linear}) {cacheName=preJacketCache}
   n11 = Get Head Up() -> (Head Up:double)
   n12 = "Transform (Modify) Bone - Bone: dyn_collar_bk"(
    ComponentPose=Local To Component(LocalPose=Copy Pose From Mesh() {bUseAttachedParent=true})
    Alpha=@n11
    Translation="0,0,0"
    Rotation="30,0.000,0.000"
    Scale="1,1,1"
   ) {
    rotationMode=BMM_Additive
    rotationSpace=BCS_BoneSpace
    alphaBoolBlend {blendOption=Linear}
    boneToModify {boneName=dyn_collar_bk}
    rotation {pitch=30}
   }
   n13 = "Transform (Modify) Bone - Bone: dyn_collar_l"(ComponentPose=@n12, Alpha=@n11, Translation="0,0,0", Rotation="15.000,5,0.000", Scale="1,1,1") {
    rotationMode=BMM_Additive
    rotationSpace=BCS_BoneSpace
    alphaBoolBlend {blendOption=Linear}
    boneToModify {boneName=dyn_collar_l}
    rotation {pitch=15, yaw=5}
   }
   n14 = "Transform (Modify) Bone - Bone: dyn_collar_r"(ComponentPose=@n13, Alpha=@n11, Translation="0,0,0", Rotation="15.000,-5,0", Scale="1,1,1") {
    rotationMode=BMM_Additive
    rotationSpace=BCS_BoneSpace
    alphaBoolBlend {blendOption=Linear}
    boneToModify {boneName=dyn_collar_r}
    rotation {pitch=15, yaw=-5}
   }
   n15 = "Constraint - dyn_grenade_01_CNST (PP)"!(ComponentPose=@n14, Alpha=1, ConstraintWeights_0=0.4, ConstraintWeights_1:float=0.6) -> (Pose) {
    alphaBoolBlend {blendOption=Linear}
    boneToModify {boneName=dyn_grenade_01_CNST}
    constraintSetup [{targetBone {boneName=spine_05}}, {targetBone {boneName=dyn_belt_shoulder_l}}]
    constraintWeights [1,1]
   }
   n16 = "Constraint - dyn_grenade_02_CNST (PP)"!(ComponentPose=@n15, Alpha=1, ConstraintWeights_0=0.35, ConstraintWeights_1:float=0.65) -> (Pose) {
    alphaBoolBlend {blendOption=Linear}
    boneToModify {boneName=dyn_grenade_02_CNST}
    constraintSetup [{targetBone {boneName=spine_05}}, {targetBone {boneName=dyn_belt_shoulder_l}}]
    constraintWeights [1,1]
   }
   n17 = Get arm_up_l() -> (arm_up_l:double)
   n18 = "Transform (Modify) Bone - Bone: dyn_grenade_01_MSTR"(ComponentPose=@n16, Alpha=@n17, Translation="0,4,-0.000", Rotation="0,0,0", Scale="1,1,1") {
    translationMode=BMM_Additive
    translationSpace=BCS_BoneSpace
    alphaBoolBlend {blendOption=Linear}
    boneToModify {boneName=dyn_grenade_01_MSTR}
    translation {y=4}
   }
   n19 = "Transform (Modify) Bone - Bone: dyn_grenade_02_MSTR"(ComponentPose=@n18, Alpha=@n17, Translation="0.000,4,0.000", Rotation="0,0,0", Scale="1,1,1") {
    translationMode=BMM_Additive
    translationSpace=BCS_BoneSpace
    alphaBoolBlend {blendOption=Linear}
    boneToModify {boneName=dyn_grenade_02_MSTR}
    translation {y=4}
   }
   n20 = "Transform (Modify) Bone - Bone: dyn_belt_shoulder_l_MSTR"(ComponentPose=@n19, Alpha=@n17, Translation="0,0,0", Rotation="0.000,10.000,0.000", Scale="1.000,0.5,1.000") {
    rotationMode=BMM_Additive
    rotationSpace=BCS_BoneSpace
    scaleMode=BMM_Additive
    scaleSpace=BCS_BoneSpace
    alphaBoolBlend {blendOption=Linear}
    boneToModify {boneName=dyn_belt_shoulder_l_MSTR}
    rotation {yaw=10.0002}
    scale {y=0.5}
   }
   n21 = "Transform (Modify) Bone - Bone: dyn_jacket_armpit_r_MSTR"(ComponentPose=@n20, Alpha=@n17, Translation="8.000,0,2", Rotation="0.000,-25.000,0", Scale="1,1,1") {
    rotationMode=BMM_Additive
    rotationSpace=BCS_BoneSpace
    translationMode=BMM_Additive
    translationSpace=BCS_BoneSpace
    alphaBoolBlend {blendOption=Linear}
    boneToModify {boneName=dyn_jacket_armpit_r_MSTR}
    rotation {yaw=-25}
    translation {x=8, z=2}
   }
   n22 = "Constraint - dyn_shoulderPad_r_02 (P)"(ComponentPose=@n21, Alpha=1, ConstraintWeights_0=1) {
    alphaBoolBlend {blendOption=Linear}
    alphaScaleBiasClamp {bInterpResult=true, interpSpeedDecreasing=5, interpSpeedIncreasing=7}
    boneToModify {boneName=dyn_shoulderPad_r_02}
    constraintSetup [{targetBone {boneName=upperarm_twist_01_r}}]
    constraintWeights [1]
   }
   n23 = "Constraint - dyn_shoulderPad_r_03 (P)"(ComponentPose=@n22, Alpha=1, ConstraintWeights_0=1) {
    alphaBoolBlend {blendOption=Linear}
    alphaScaleBiasClamp {bInterpResult=true, interpSpeedDecreasing=5, interpSpeedIncreasing=7}
    boneToModify {boneName=dyn_shoulderPad_r_03}
    constraintSetup [{targetBone {boneName=upperarm_twist_01_r}}]
    constraintWeights [1]
   }
   n24 = "Transform (Modify) Bone - Bone: dyn_shoulderPad_r_01"!(ComponentPose=@n23, Alpha=1, Rotation="15,5,0", Scale="0.85,1,1") -> (Pose) {
    rotationMode=BMM_Additive
    rotationSpace=BCS_BoneSpace
    scaleMode=BMM_Additive
    scaleSpace=BCS_BoneSpace
    alphaBoolBlend {blendOption=Linear}
    alphaScaleBiasClamp {bInterpResult=true, interpSpeedDecreasing=5, interpSpeedIncreasing=7}
    boneToModify {boneName=dyn_shoulderPad_r_01}
    rotation {pitch=15, yaw=5}
    scale {x=0.85}
   }
   n25 = "Transform (Modify) Bone - Bone: dyn_shoulderPad_l_01"!(ComponentPose=@n24, Alpha=1, Scale="0.85,1,1") -> (Pose) {
    rotationSpace=BCS_BoneSpace
    scaleMode=BMM_Additive
    scaleSpace=BCS_BoneSpace
    alphaBoolBlend {blendOption=Linear}
    alphaScaleBiasClamp {bInterpResult=true, interpSpeedDecreasing=5, interpSpeedIncreasing=7}
    boneToModify {boneName=dyn_shoulderPad_l_01}
    rotation {pitch=15, yaw=5}
    scale {x=0.85}
   }
   Save cached pose 'PrePose'(Pose=Component To Local(ComponentPose=@n25)) {cacheName=PrePose}
  ]}
 EventGraph {nodes [
   Event BlueprintUpdateAnimation()
   > n1 = Sequence!(execute) -> (then_0, then_1, then_2:Exec, then_3:Exec)
   n2 = Get bIsPlayingEmote()
   n3 = Get bIsInFrontEnd()
   Set Front Emote(execute=@n1.then_1, Front Emote:double=SelectFloat(A=1, B=0, bPickA=AND Boolean(A=@n2, B=@n3))) -> (Output_Get:double)
   n4 = Branch(execute=@n1.then_3, Condition=Get bIsSkyDiving())
   > n5 = Branch(Condition=Get bIsParachuteOpened())
   Set gravity_override(execute=[@n5.then, @n4.else], gravity_override="0, 0,-980")
   Set gravity_override(execute=@n5.else, gravity_override="0, 0,-300")
   n6 = BreakTransform(InTransform=GetDeltaTransformFromRefPose(self=GetOwnerSkeletalMeshComponent(), BoneName=head, BaseName=spine_05))
   Set Head Up(execute=@n1.then_0, Head Up:double=MapRangeClamped(Value=BreakRotator(InRot=@n6.Rotation).Pitch, InRangeA=15, InRangeB=35)) -> (Output_Get:double)
   n7 = BreakTransform(InTransform=GetDeltaTransformFromRefPose(self=GetOwnerSkeletalMeshComponent(), BoneName=upperarm_l, BaseName=spine_05))
   > Set arm_up_l(arm_up_l:double=MapRangeClamped(Value=BreakRotator(InRot=@n7.Rotation).Yaw, InRangeA=50, InRangeB=120)) -> (Output_Get:double)
   n8 = BreakTransform(InTransform=GetDeltaTransformFromRefPose(self=GetOwnerSkeletalMeshComponent(), BoneName=upperarm_l, BaseName=lowerarm_l))
   > Set elbow_rotate_l(elbow_rotate_l:double=MapRangeClamped(Value=BreakRotator(InRot=@n8.Rotation).Yaw, InRangeA=0, InRangeB=-70)) -> (Output_Get:double)
   n9 = BreakTransform(InTransform=GetDeltaTransformFromRefPose(self=GetOwnerSkeletalMeshComponent(), BoneName=upperarm_r, BaseName=lowerarm_r))
   > Set elbow_rotate_r(elbow_rotate_r:double=MapRangeClamped(Value=BreakRotator(InRot=@n9.Rotation).Yaw, InRangeA=0, InRangeB=-70)) -> (Output_Get:double)
   n10 = Get bIsTargeting()
   n11 = Get bIsCrouching()
   Set rigid alpha(execute=@n1.then_2, rigid alpha:double=SelectFloat(A=0.3, B=1, bPickA=OR Boolean(A=@n10, B=@n11))) -> (Output_Get:double)
   Event OnCharacterCustomizationLoaded()
   > Do Once()
   n12 = CheckForBackbling()
   > Set isWearingBackbling(isWearingBackbling=OR Boolean(A=@n12, B=@n12.bIsBackblingLong))
   > Branch(Condition=@n12)
   > Set isWearingCape(isWearingCape=@n12.bIsBackblingLong)
   n13 = GetPartSkeletalMeshComponent()
   n14 = FindChildCommponentFromAttachSocket!(
    self:Custom Character Part Anim Instance Object Reference
    Parent:Skeletal Mesh Component Object Reference=@n13
    bIncludeAllDescendants:bool=false
    AttachSocket:Name=Cape
   ) -> (ReturnValue:Skeletal Mesh Component Object Reference)
   "Not Equal (Object)"!(A:Object Reference=@n14, B:Object Reference) -> (ReturnValue:bool)
   "float > float"(
    A=Get Distance Between Two Sockets!(
     Component:Skeletal Mesh Component Object Reference=@n14
     SocketOrBoneNameA:Name=backpack_root
     SocketSpaceA:ERelativeTransformSpace Enum=RTS_World
     SocketOrBoneNameB:Name=cape_length
     SocketSpaceB:ERelativeTransformSpace Enum=RTS_World
     bRemapRange:bool=false
     InRangeMin:float=0
     InRangeMax:float=0
     OutRangeMin:float=0
     OutRangeMax:float=0
    ) -> (ReturnValue:float) {is=title K2Node_CallFunction}
    B=50
   ) {is=title K2Node_CallFunction}
   "Not Equal (Object)"(A=FindChildCommponentFromAttachSocket(Parent=@n13, bIncludeAllDescendants=false, AttachSocket=backpack_br))
  ]}
]
variables [
 arm_up_l {type=number}
 elbow_rotate_l {type=number}
 elbow_rotate_r {type=number}
 Front Emote {type=number}
 gravity_override {type=Vector}
 Head Up {type=number}
 isWearingBackbling {type=boolean, default=true}
 isWearingCape {type=boolean}
 rigid alpha {type=number}
]
