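# Physics asset for the face accessory of the Sleepy Time Ninja male medium player cosmetic, covering the dynamic bandanna bones.
# Three capsule bodies cover the knot and the two dyn_bandanna bones attached to it.
# Each of those two bones has its own constraint to the knot, with wide swing limits and a tight twist limit.
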PhysicsAsset /BRCosmetics/Characters/Player/Male/Medium/Bodies/M_MED_SleepyTimeNinja/Meshes/Parts/M_MED_Sleepy_Time_Ninja_FaceAcc_Dyn
bodies {
 dyn_bandanna_01 [{shapeType=Capsule, length=2.5, radius=1.5, center=<1 0 0>, rotation=<rot 90 0 0>}]
 dyn_bandanna_02 [{shapeType=Capsule, length=2.5, radius=1.5, center=<1 0 0>, rotation=<rot 90 0 0>}]
 dyn_bandanna_knot [{shapeType=Capsule, length=1, radius=1.5, rotation=<rot 90 0 0>}]
}
constraints [
 {bone1Name=dyn_bandanna_01, bone2Name=dyn_bandanna_knot, swing1LimitDegrees=75, swing2LimitDegrees=90, twistLimitDegrees=5}
 {bone1Name=dyn_bandanna_02, bone2Name=dyn_bandanna_knot, swing1LimitDegrees=75, swing2LimitDegrees=90, twistLimitDegrees=5}
]
